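# Niagara swing trail for the Angel pickaxe, combining two blue anim-trail ribbons with sparkle sprites.
# MainTrail's alpha follows the system SwingGeneralFade, while the narrower SecondaryTrail has two ribbon renderers and animates alternate color and dynamic material parameters over life.
# Both ribbons set their RibbonID from an emitter offset.
# Particles_Swing spawns per distance moved during a short delayed window, sampling a socket, then drifts and orbits with pulsing size.
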NiagaraSystem /BRCosmetics/Weapons/FORT_Melee/Pickaxe_Angel/FX_FNCS/NS_Pickaxe_Angel_Trail
dependencies {
 usedDataInterfaces [&NiagaraDataInterfaceCurve, &NiagaraDataInterfaceVectorCurve, &NiagaraDataInterfaceSocketReader]
 usedDynamicInputs []
 usedModules [
  &SystemState
  &self:SystemSpawnScript.SystemScriptSource.SystemScriptGraph.NiagaraNodeAssignment_0.SetVariables_#1
  &self:SystemSpawnScript.SystemScriptSource.SystemScriptGraph.NiagaraNodeAssignment_1.SetVariables_#2
 ]
}
emitters {
 MainTrail {
  emitterSpawnScript [&FN_RibbonIDOffsetUtility()]
  emitterUpdateScript [
   &EmitterState()
   &FN_SpawnAnimTrail(AnimTrailWidthMode=::FromSecond)
  ]
  particleSpawnScript [
   &FN_InitializeAnimTrailBoneIndecies()
   &FN_InitializeAnimTrail(AnimTrailWidth=2, Color=<rgba 0 0.5049 1.5 1>, Lifetime=0.3)
   /SetVariables(Particles.RibbonID=/Niagara/DynamicInputs/NiagaraID/MakeNiagaraIDFromInt(Index:NiagaraInt32=$Emitter.RibbonIDOffset, Acquire Tag:NiagaraInt32=0)) {isSetParameters=true, module=SetVariables_#3}
  ]
  particleUpdateScript [
   &FN_UpdateAnimTrail() {module=FN_UpdateAnimTrail001}
   &ScaleColor(
    Scale Alpha=&FloatFromCurve(
     FloatCurve=DataInterface() {propertyValues {Curve {keys [<at 0.5 1>, {arriveTangent=-2.9197, time=1, value=0}]}}}
     Scale Curve=$System.SwingGeneralFade
    )
    ScaleA=true
    ScaleRGB=true
   )
   /SetVariables(Particles.ScaledRibbonWidth:NiagaraFloat=&Multiply_Float(
     A=$Particles.Initial.RibbonWidth
     B=&FloatFromCurve(FloatCurve=DataInterface() {propertyValues {Curve {keys [<at 0 1>,<at 1 0.7>]}}})
    )) {isSetParameters=true, module=SetVariables_#4}
  ]
  rendererClasses [&NiagaraRibbonRendererProperties]
  renderers [{rendererClass=&NiagaraRibbonRendererProperties}]
 }
 Particles_Swing {
  emitterUpdateScript [
   &EmitterState(Life Cycle Mode=::NewEnumerator1, Loop Behavior=::NewEnumerator1, Loop Delay=0.08, Loop Duration=0.27, UseLoopDelay=true)
   &SpawnPerUnit(
    Max Movement Threshold=2500
    Module.Use Max Movement Threshold=false
    Module.Use Movement Tolerance=false
    Spawn Spacing=25
    Use Spawn Probability=false
   )
  ]
  particleSpawnScript [
   &InitializeParticle(
    Color=<rgba 0 0.0912 1 1>
    Lifetime=&RandomRangeFloat(Minimum=0.5, Maximum=1, Evaluation Type=::NewEnumerator1)
    Sprite Rotation Mode=::NewEnumerator2
    Sprite Size Mode=::NewEnumerator3
    Uniform Sprite Size=&RandomRangeFloat(Minimum=2.5, Maximum=5, Evaluation Type=::NewEnumerator1)
    UsePositionOffset=false
    Write Material Random=true
   )
   &SocketLocation(
    Interpolate Socket Sampling=true
    SamplePosOffset=false
    Socket Index=1
    Socket Reader=$System.Socket Reader
    Socket Sampling Mode=::NewEnumerator3
    VelocitySpeedLimit=false
   )
   &ShapeLocation(MaskSpawnGroup=false, Override Seed=false, Sphere Radius=10, UseOwnerScale=true)
   &AddVelocity(
    Module.Use Falloff Distance=false
    Override Seed=false
    Velocity Mode=::NewEnumerator1
    Velocity Origin=$Output.SocketLocation.SampledPosition
    Velocity Speed=5
   )
   &InheritVelocity(
    Inherited Velocity Speed Limit=20
    Speed Limit=true
    UseSpeedThreshold=false
    Velocity Source=$Particles.SkeletalMeshLocation.SampledVelocity
   )
  ]
  particleUpdateScript [
   &ParticleState()
   &Drag(Drag=2, Use Linear Drag=true, Use Rotational Drag=true)
   &ScaleColor(
    Scale Alpha=&FloatFromCurve(FloatCurve=DataInterface() {propertyValues {Curve {keys [<at 0 1>,<at 0.6 1>,<at 1 0>]}}})
    Scale RGB=<0 50 50>
    ScaleA=true
    ScaleRGB=true
   )
   &ScaleSpriteSize(
    Scale Sprite Size Mode=::NewEnumerator0
    Uniform Scale Factor=&FloatFromCurve(FloatCurve=DataInterface() {propertyValues {Curve {keys [
         {arriveTangent=27.05, interpMode=RCIM_Cubic, leaveTangent=27.05, tangentMode=RCTM_User, time=0, value=0}
         {arriveTangent=1.3101, interpMode=RCIM_Cubic, leaveTangent=1.3101, time=0.0546, value=1.2477}
         {arriveTangent=-1.3446, interpMode=RCIM_Cubic, leaveTangent=-1.3446, time=0.1581, value=0.2072}
         {arriveTangent=0.1974, interpMode=RCIM_Cubic, leaveTangent=0.1974, time=0.2882, value=0.9336}
         {arriveTangent=0.2295, interpMode=RCIM_Cubic, leaveTangent=0.2295, time=0.4301, value=0.2608}
         {arriveTangent=-0.3367, interpMode=RCIM_Cubic, leaveTangent=-0.3367, time=0.5884, value=1.0025}
         {arriveTangent=-1.2837, interpMode=RCIM_Cubic, leaveTangent=-1.2837, time=0.7448, value=0.1549}
         {arriveTangent=-0.607, interpMode=RCIM_Cubic, leaveTangent=-0.607, time=0.834, value=0.6872}
         <at 1 0 RCIM_Cubic>
        ]}}})
   )
   &SpriteRotationRate(Rotation Rate=&RandomRangeFloat(Minimum=-0.5, Maximum=0.5))
   &SolveForcesAndVelocity(Clamp Velocity=false, Limit Acceleration=false)
   /Niagara/Modules/CascadeConversion/CascadeConversion_Orbit(
    Offset:Vector3f=&RandomRangeVector(Minimum=<0 0 0>, Maximum=<0 10 0>)
    Rotation:Vector3f=&RandomRangeVector(Minimum=<0 0 0>, Maximum=<0.5 0.5 0.5>)
    Rotation Rate:Vector3f=&RandomRangeVector(Minimum=<0 0 0>, Maximum=<1 1 1>)
   )
   /Niagara/Modules/CascadeConversion/CascadeConversion_SolveOrbit()
  ]
  rendererClasses [&NiagaraSpriteRendererProperties]
  renderers [{rendererClass=&NiagaraSpriteRendererProperties}]
 }
 SecondaryTrail {
  emitterSpawnScript [&FN_RibbonIDOffsetUtility()]
  emitterUpdateScript [
   &EmitterState()
   &FN_SpawnAnimTrail(AnimTrailWidthMode=::FromFirst)
  ]
  particleSpawnScript [
   &FN_InitializeAnimTrailBoneIndecies(First Bone or Socket Index=2, Second Bone or Socket Index=0)
   &FN_InitializeAnimTrail(AnimTrailWidth=0.8, Color=<rgba 0 0.1352 0.474 13>)
   /SetVariables(Particles.RibbonID=&MakeNiagaraIDFromInt(Index:NiagaraInt32=$Emitter.RibbonIDOffset, Acquire Tag:NiagaraInt32=0)) {isSetParameters=true, module=SetVariables_#3}
  ]
  particleUpdateScript [
   &FN_UpdateAnimTrail() {module=FN_UpdateAnimTrail001}
   &ScaleColor(
    Scale Alpha=&FloatFromCurve(FloatCurve=DataInterface() {propertyValues {Curve {keys [
         {arriveTangent=-1.0274, interpMode=RCIM_Cubic, leaveTangent=-5.28, tangentMode=RCTM_Break, time=0.0008, value=-0.0015}
         {arriveTangent=-2.0949, interpMode=RCIM_Cubic, leaveTangent=-2.0949, tangentMode=RCTM_Break, time=0.1009, value=0.4933}
         {arriveTangent=-0.5487, interpMode=RCIM_Cubic, leaveTangent=-0.5487, time=0.4181, value=0.1241}
         <at 1 0>
        ]}}})
    ScaleA=true
    ScaleRGB=true
   )
   &ScaleRibbonWidth(Ribbon Width Scale=&FloatFromCurve(FloatCurve=DataInterface() {propertyValues {Curve {keys [<at 0 1>,<at 1 0.7>]}}}))
   /SetVariables(Particles.ScaledRibbonWidth:NiagaraFloat=&Multiply_Float(
     A=$Particles.Initial.RibbonWidth
     B=&FloatFromCurve(FloatCurve=DataInterface() {propertyValues {Curve {keys [<at 0 1>,<at 1 0.7>]}}}, Scale Curve=0.7)
    )) {isSetParameters=true, module=SetVariables_#4}
   /SetVariables(Particles.ColorAlt:LinearColor=&MultiplyLinearColor(
     Linear Color=<rgba 0.001 0.4999 1 2>
     Scale Factor=&MakeVector4FromVectorAndFloat(
      "Vector (RGB)"=&VectorFromCurve(VectorCurve=DataInterface() {propertyValues {XCurve {keys [<at 0 0>,<at 1 3>]}, YCurve {keys [<at 0 10>,<at 1 0>]}, ZCurve {keys [<at 0 7.8944>,<at 1 10>]}}})
      "Float (Alpha)"=&FloatFromCurve(FloatCurve=DataInterface() {propertyValues {Curve {keys [
           {arriveTangent=-1.0274, interpMode=RCIM_Cubic, leaveTangent=-5.28, tangentMode=RCTM_Break, time=-0.0008, value=-0.0015}
           {arriveTangent=4.0283, interpMode=RCIM_Cubic, leaveTangent=4.0283, time=0.0668, value=-0.0008}
           {arriveTangent=0.3555, interpMode=RCIM_Cubic, leaveTangent=0.3555, time=0.0994, value=0.4019}
           {arriveTangent=-0.4462, interpMode=RCIM_Cubic, leaveTangent=-0.4462, time=0.4181, value=0.1241}
           <at 1 0>
          ]}}})
     )
    )) {isSetParameters=true}
   &DynamicMaterialParameters(
    Dynamic Parameter Output:ENiagara_AttributeSamplingApplyOutput=::NewEnumerator1
    Index 0 Param 1=&FloatFromCurve(FloatCurve=DataInterface() {propertyValues {Curve {keys [{arriveTangent=-1.5181, time=0, value=1.5}, <at 0.9344 -0.0181>]}}})
    Index 0 Param 2=&FloatFromCurve(FloatCurve=DataInterface() {propertyValues {Curve {keys [<at 0 0>,<at 0.5 0.4>,<at 0.82 0>]}}})
    Index 0 Param 3=1
    Index 0 Param 4=1
    Module.Index0Float0Write=true
    Module.Index0Float1Write=true
    Module.Index0Float2Write=true
    Module.Index0Float3Write=true
   )
  ]
  rendererClasses [&NiagaraRibbonRendererProperties]
  renderers [
   {rendererClass=&NiagaraRibbonRendererProperties}
   {rendererClass=&NiagaraRibbonRendererProperties, rendererIndex=1}
  ]
 }
}
summary {
 systemName=NS_Pickaxe_Angel_Trail
 emitters [
  {emitterName=MainTrail, rendererClasses [&NiagaraRibbonRendererProperties]}
  {emitterName=Particles_Swing, rendererClasses [&NiagaraSpriteRendererProperties]}
  {emitterName=SecondaryTrail, rendererClasses [&NiagaraRibbonRendererProperties]}
 ]
 userVariables [User.RibbonIDOffset {type=NiagaraInt32}]
}
